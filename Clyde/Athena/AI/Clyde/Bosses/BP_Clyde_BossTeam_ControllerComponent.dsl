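; Boss team controller component that assigns spawned pawns to the team's factions and sets faction attitude toward them.
; AddFactionTag extends the stored faction tag set when the CanSetFactions scalable float is true.
; On pawn spawn, it adds the pawn to those factions, then applies the stored attitude from those factions toward it.
; Scalable float toggles switch that attitude to Friendly only when CanSetNeutral is off and CanSetFriendly is on.
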(event EventBeginPlay)

(event EventTick
  (params (DeltaSeconds :type Float%20%28single-precision%29)))

(event Custom|AddFactionTag
  (params (GameplayTag :type Gameplay%20Tag%20Structure))
  (if (ScalableFloat|GetValueasBool (Variables|Default|GetCanSetFactions))
    (GameplayTags|AddGameplayTag (Variables|Default|GetFactionTags) GameplayTag :!class /Script/GameplayTags.BlueprintGameplayTagLibrary)))

(event EventOnPawnSpawned
  (params (Pawn :type Pawn%20Object%20Reference))
  (Affiliation|AddtoFactions Pawn (Variables|Default|GetFactionTags))
  (if (ScalableFloat|GetValueasBool (Variables|Default|GetCanSetNeutral))
    _
    (elif (ScalableFloat|GetValueasBool (Variables|Default|GetCanSetFriendly))
      (Variables|Default|SetFactionAttitude "Friendly")))
  (Affiliation|ChangeFactionAttitudeTowardsActor (Variables|Default|GetFactionTags) Pawn (Variables|Default|GetFactionAttitude)))
